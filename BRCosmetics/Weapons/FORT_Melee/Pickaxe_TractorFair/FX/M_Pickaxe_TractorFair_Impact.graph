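# Translucent, unlit, two-sided Niagara sprite material for the TractorFair pickaxe impact effect.
# Rotated UVs drive a panning noise that distorts a shape mask, weighted by a gradient along one axis.
# The mask forms a particle-tinted emissive value and a contrasted opacity.
# A second noise, offset by a gradient and the Dynamic Parameter Dissolve value, erodes that opacity before particle alpha scales it.
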Material /BRCosmetics/Weapons/FORT_Melee/Pickaxe_TractorFair/FX/M_Pickaxe_TractorFair_Impact
graph {
 properties {
  bUsedWithNiagaraSprites=true
  blendMode=BLEND_Translucent
  pixelDepthOffsetMode=PDOM_Legacy
  shadingModel=MSM_Unlit
  twoSided=true
  usedShadingModels=Unlit
 }
 nodes [
  n1 = MaterialFunctionCall(Float2=TextureCoordinate()) {materialFunction=&BreakOutFloat2Components}
  n2 = MaterialFunctionCall(X=OneMinus(None=@n1.G), Y=@n1.R) {materialFunction=&MakeFloat2}
  CurveAtlasRowParameter(CurveTime=MaterialFunctionCall(Float2=@n2.Result) {materialFunction=&BreakOutFloat2Components}.R) {
   atlas=/Game/Weapons/FORT_Melee/Pickaxe_InstantGravelNoble/Meshes/Test/CLA_InstantGraveNoble
   curve=/Game/Weapons/FORT_Melee/Pickaxe_InstantGravelNoble/Meshes/Test/CL_ImpactGrad
   defaultValue=3
  }
  n3 = TextureObjectParameter(UVs=@n2.Result) {
   group=Mask
   parameterName=Mask_Texture
   sortPriority=0
   texture=/BRCosmetics/Weapons/FORT_Melee/Pickaxe_InstantGravelNoble/Meshes/Test/T_ImpactShape
  }
  n4 = TextureObjectParameter(UVs=@n2.Result) {
   group=MainNoise
   parameterName=MainNoise_Texture
   samplerType=SAMPLERTYPE_LinearGrayscale
   sortPriority=0
   texture=/Game/Global/Textures/Default/Tileable_Patterns/T_StormCloudDetail01_M
  }
  n5 = VectorParameter() {group=MainNoise, parameterName=MainNoise_UVControl, channelNames {a=VPan, b=UPan, g=VScale, r=UScale}, defaultValue=<rgba 0.5 1 -1.4835 0>}
  n6 = MaterialFunctionCall(Texture=@n4, UVControl_RGB=@n5, UVControl_A=@n5, "[Override]UVInput"=@n2.Result) {materialFunction=&FXCommon_TextureSample}
  n7 = VectorParameter() {
   group=MainNoise
   parameterName=MainNoise_NoiseControl
   sortPriority=60
   channelNames {a=VStrength, b=UStrength, g=VOffset, r=UOffset}
   defaultValue=<rgba -0.772 -0.5 0 0.464>
  }
  n8 = Power(
   Base=OneMinus(None=MaterialFunctionCall(Float2=@n2.Result) {materialFunction=&BreakOutFloat2Components}.R)
   Exp=ScalarParameter() {defaultValue=2, parameterName=W}
  )
  n9 = VectorParameter() {group=Mask, parameterName=Mask_UVControl, channelNames {a=VPan, b=UPan, g=VScale, r=UScale}, defaultValue=<rgba 1 1 0 0>}
  n10 = MaterialFunctionCall(
   Texture=@n3
   AdditionalUVOffset=Multiply(
    A=MaterialFunctionCall!(NoiseValue=@n6.RG, Control_RGB=@n7, Control_A=@n7) {materialFunction=/Game/Effects/Niagara/MasterMaterials/Functions/Common/FXCommon_NoiseControl}.NoiseValue
    B=OneMinus(None=@n8)
   )
   UVControl_RGB=@n9
   UVControl_A=@n9
   "[Override]UVInput"=@n2.Result
  ) {materialFunction=&FXCommon_TextureSample}
  n11 = Saturate(None=@n10.R)
  n12 = ParticleColor()
  n13 = Multiply(A=LinearInterpolate(A=ScalarParameter() {parameterName=Param_1}, Alpha=Multiply(B=@n11) {constA=1}), B=@n12.RGB)
  Divide(A=@n13) {constB=1.7}
  MaterialFunctionCall(Emissive=@n13) {materialFunction=&MF_FX_EmissiveAdjustments}
  n14 = Multiply(
   A=@n10.R
   B=MaterialFunctionCall!(TexCoord=@n10.UV, ClampChannel=Constant2Vector() {g=1, r=1}) {materialFunction=/Game/Effects/Fort_Effects/MaterialFunctions/MF_FX_TilingMethod}.Result
  )
  n15 = Constant() {r=0.2}
  n16 = Saturate(None=Multiply(
    A=MaterialFunctionCall(In=@n14, Contrast=@n15) {materialFunction=&CheapContrast}.Result
    B=ScalarParameter() {defaultValue=5, parameterName=ScaleOpacity}
   ))
  n17 = TextureObjectParameter(UVs=@n2.Result) {
   group=DissolveTexture
   parameterName=DissolveTexture_Texture
   samplerType=SAMPLERTYPE_LinearGrayscale
   sortPriority=0
   texture=/Game/Effects/Fort_Effects/Emotes/UltraLight/Texture/T_Plasma_A_B_Greyscale
  }
  n18 = VectorParameter() {
   group=DissolveTexture
   parameterName=DissolveTexture_UVControl
   channelNames {a=VPan, b=UPan, g=VScale, r=UScale}
   defaultValue=<rgba 0.3 0.7 0.1 0>
  }
  n19 = Multiply(
   A=OneMinus(None=MaterialFunctionCall(Float2=@n2.Result) {materialFunction=&BreakOutFloat2Components}.R)
   B=ScalarParameter() {defaultValue=2.5, parameterName=DissolveGrad}
  )
  n20 = Add(
   A=ScalarParameter() {defaultValue=-2.2512, parameterName=DissolvePer}
   B=DynamicParameter() {paramNames [Dissolve,Param2,Param3,Param4]}.Dissolve
  )
  n21 = Multiply(
   A=Subtract(
    A=MaterialFunctionCall(Texture=@n17, UVControl_RGB=@n18, UVControl_A=@n18, "[Override]UVInput"=@n2.Result) {materialFunction=&FXCommon_TextureSample}.R
    B=Add(A=@n19, B=@n20)
   )
   B=ScalarParameter() {defaultValue=10, parameterName=DissolveMul}
  )
  Multiply(A=@n12.A, B=Saturate(None=Multiply(A=@n16, B=@n21)))
  OneMinus(None=@n11)
  OneMinus(None=@n10.R)
 ]
}
