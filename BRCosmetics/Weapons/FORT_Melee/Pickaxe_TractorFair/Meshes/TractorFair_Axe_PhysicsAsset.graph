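# Physics asset for the TractorFair two-handed pickaxe mesh, a cosmetic melee weapon.
# It holds a single capsule body on the Melee_TwoHanded_Handle bone, offset from the bone origin and pitched or rotated to fit the shape.
# There are no constraints, so the pickaxe collides as one rigid piece.
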PhysicsAsset /BRCosmetics/Weapons/FORT_Melee/Pickaxe_TractorFair/Meshes/TractorFair_Axe_PhysicsAsset
bodies {Melee_TwoHanded_Handle [{shapeType=Capsule, length=33.0256, radius=11.1671, center=<10.3869 -0.0985 13.4107>, rotation=<rot -0.2434 -90 0.7133>}]}
constraints []
